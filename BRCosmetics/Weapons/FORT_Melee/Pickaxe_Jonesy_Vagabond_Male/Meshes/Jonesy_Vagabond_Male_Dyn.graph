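# Physics asset for the Jonesy Vagabond Male pickaxe, a two-handed melee cosmetic with two dynamic donut bones.
# A capsule body covers part of the handle, and a box body covers each of the inner and outer donut bones.
# Each donut is constrained directly to the handle rather than to the other, with identical angular limits that are much wider on one swing axis than the other.
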PhysicsAsset /BRCosmetics/Weapons/FORT_Melee/Pickaxe_Jonesy_Vagabond_Male/Meshes/Jonesy_Vagabond_Male_Dyn
bodies {
 Melee_TwoHanded_Handle [{shapeType=Capsule, length=12.5, radius=4.0625, center=<0.0001 2.1396 31.8505>, rotation=<rot 0 0 -0.5213>}]
 dyn_donut_in [{shapeType=Box, extentX=8.8125, extentY=2.6875, extentZ=8.8125, center=<2.2713 -0.2098 0>, rotation=<rot 0 -5.001 0>}]
 dyn_donut_out [{shapeType=Box, extentX=8.8125, extentY=2.75, extentZ=8.8125, center=<1.7111 -0.2656 0>, rotation=<rot 0 -15.0014 0>}]
}
constraints [
 {bone1Name=dyn_donut_in, bone2Name=Melee_TwoHanded_Handle, swing1LimitDegrees=15, swing2LimitDegrees=90, twistLimitDegrees=25}
 {bone1Name=dyn_donut_out, bone2Name=Melee_TwoHanded_Handle, swing1LimitDegrees=15, swing2LimitDegrees=90, twistLimitDegrees=25}
]
